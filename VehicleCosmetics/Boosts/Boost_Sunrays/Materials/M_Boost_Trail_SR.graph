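# Translucent unlit material for the Sunrays vehicle boost trail, built for Niagara ribbons.
# Panning noise warps a soft cross-ribbon mask and two scrolling ray-gradient layers, whose multiply-to-add blend is driven by a dynamic parameter and carves the opacity, faded at both ends.
# Particle color feeds emissive through optional hue shift and desaturation.
# Static switches add distortion and sine/cosine world position offset wobble.
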Material /VehicleCosmetics/Boosts/Boost_Sunrays/Materials/M_Boost_Trail_SR
graph {
 properties {
  bUsedWithNiagaraRibbons=true
  blendMode=BLEND_Translucent
  pixelDepthOffsetMode=PDOM_Legacy
  refractionMethod=RM_2DOffset
  shadingModel=MSM_Unlit
  usedShadingModels=Unlit
 }
 nodes [
  n1 = ParticleColor()
  n2 = Constant() {desc=.3, r=0.5}
  n3 = TextureCoordinate() {vTiling=0.125}
  n4 = Constant() {r=0.5}
  n5 = TextureSample(UVs=Panner(Coordinate=TextureCoordinate() {uTiling=0.25, vTiling=0.25}) {speedX=-0.125, speedY=0.01}) {texture=/DealershipSharedLibrary/FX/Textures/Noise/T_VCVFX_Flow02_N}
  n6 = Multiply(B=AppendVector(A=@n5.R, B=@n5.R)) {constA=0.05}
  n7 = Panner(Coordinate=Add(A=Panner(Coordinate=@n3, Time=@n4) {speedY=0.5}, B=@n6)) {speedX=-1}
  n8 = TextureCoordinate() {uTiling=1.25, vTiling=-0.125}
  n9 = Panner(Coordinate=Add(A=Panner(Coordinate=@n8, Time=@n4) {speedY=-0.5}, B=@n6)) {speedX=-1}
  n10 = ComponentMask(None=Add(
    A=TextureSample(UVs=@n7) {texture=/DealershipSharedLibrary/FX/Textures/Noise/T_VCVFX_DistortNoise02_RGB}.RGB
    B=TextureSample(UVs=@n9) {texture=&T_VCVFX_DistortNoise02_RGB}.RGB
   )) {g=true, r=true}
  n11 = ComponentMask(None=Add(A=TextureCoordinate(), B=Multiply(A=@n2, B=@n10) {constA=0.5})) {g=true}
  n12 = OneMinus(None=@n11)
  n13 = Multiply(A=Clamp(None=ConstantBiasScale(None=@n11) {bias=-0.15, scale=1}), B=Clamp(None=ConstantBiasScale(None=@n12) {bias=-0.01, scale=1}))
  n14 = ScalarParameter() {defaultValue=1, parameterName=Hardness}
  n15 = Clamp(None=Multiply(A=@n13, B=@n14))
  n16 = ScalarParameter() {defaultValue=1, parameterName=Bightness}
  n17 = Multiply(A=@n15, B=@n16)
  n18 = Clamp(None=Multiply(B=@n13) {constA=6})
  n19 = ScalarParameter() {defaultValue=1, parameterName=SolidMask}
  n20 = LinearInterpolate(A=@n17, B=@n18, Alpha=@n19)
  n21 = TextureCoordinate() {uTiling=1.25, vTiling=0.75}
  n22 = TextureCoordinate() {uTiling=1.25, vTiling=0.75}
  n23 = Constant() {r=0.5}
  n24 = Multiply(A=Constant() {desc=.3, r=0.15}, B=@n10)
  n25 = Add(A=Panner(Coordinate=StaticSwitchParameter(True=@n21, False=@n22) {parameterName=UseTileDistance}, Time=@n23) {speedY=0.5}, B=@n24)
  n26 = Multiply(A=Time(), B=ScalarParameter() {defaultValue=2, parameterName=PanSpeed})
  n27 = ChannelMaskParameter(None=TextureSample(UVs=Panner(Coordinate=@n25, Time=@n26) {speedX=-1}) {texture=/VehicleCosmetics/Boosts/Boost_Sunrays/Textures/T_VCVFX_Boost_Sunrays_RaysGradient_RGB}.RGB) {maskChannel=Green, parameterName=MaskRGB_01, defaultValue=<rgba 0 1 0 0>}
  n28 = TextureCoordinate() {vTiling=-0.75}
  n29 = TextureCoordinate() {vTiling=-0.75}
  n30 = Add(A=@n24, B=Panner(Coordinate=StaticSwitchParameter(True=@n28, False=@n29) {parameterName=UseTileDistance}, Time=@n23) {speedY=-0.5})
  n31 = ChannelMaskParameter(None=TextureSample(UVs=Panner(Coordinate=@n30, Time=@n26) {speedX=-0.5}) {texture=&T_VCVFX_Boost_Sunrays_RaysGradient_RGB}.RGB) {maskChannel=Blue, parameterName=MaskRGB_02, defaultValue=<rgba 0 0 1 0>}
  n32 = Multiply(A=Multiply(A=@n27.R, B=@n31.R), B=@n16)
  n33 = Multiply(A=@n16, B=Add(A=@n27.R, B=@n31.R))
  n34 = Multiply(
   A=ScalarParameter() {defaultValue=1, parameterName=AddMultiply}
   B=DynamicParameter() {paramNames [Hardness,SolidMask,AddMultiply,add2mult]}.add2mult
  )
  n35 = Clamp(None=OneMinus(None=LinearInterpolate(A=@n32, B=@n33, Alpha=@n34)))
  n36 = Clamp(None=Subtract(A=@n20, B=Multiply(A=@n18, B=@n35)))
  n37 = LinearInterpolate(A=@n17, B=Clamp(None=Multiply(A=@n36, B=@n14)), Alpha=@n19)
  n38 = ComponentMask(None=TextureCoordinate()) {r=true}
  n39 = Clamp(None=ConstantBiasScale(None=@n38) {bias=-0.01, scale=15})
  n40 = Clamp(None=ConstantBiasScale(None=OneMinus(None=@n38)) {bias=-0.01, scale=5})
  Clamp(None=Multiply(A=Multiply(A=@n1.A, B=@n37), B=Multiply(A=@n39, B=@n40))) {maxDefault=1.25}
  n41 = ScalarParameter() {parameterName=HueShift}
  n42 = Multiply(B=@n41) {constA=0.1}
  n43 = ParticleColor()
  n44 = LinearInterpolate(A=@n41, B=@n42, Alpha=@n43.R)
  n45 = LinearInterpolate(A=@n42, B=@n41, Alpha=Add(A=@n43.G, B=@n43.G))
  n46 = StaticSwitchParameter(True=Clamp(None=Add(A=@n44, B=@n45)), False=@n41) {parameterName=UseHueOffset}
  n47 = Clamp(None=@n20)
  n48 = ScalarParameter() {parameterName=Saturation}
  n49 = Multiply(B=@n48) {constA=0.5}
  n50 = ParticleColor()
  n51 = Add(A=LinearInterpolate(A=@n48, B=@n49, Alpha=@n50.R), B=LinearInterpolate(A=@n49, B=@n48, Alpha=Add(A=@n50.G, B=@n50.G)))
  n52 = Desaturation(
   None=MaterialFunctionCall(Hue Shift Percentage=Multiply(A=@n46, B=@n47), Texture=@n1.RGB) {materialFunction=&HueShift}.Result
   Fraction=StaticSwitchParameter(True=Clamp(None=@n51), False=@n48) {defaultValue=true, parameterName=UseSaturationOffset}
  )
  MaterialFunctionCall(Emissive=Multiply(A=StaticSwitchParameter(True=@n52, False=@n1.RGB) {defaultValue=true, parameterName=UseHueShift}, B=@n20)) {materialFunction=&MF_FX_EmissiveAdjustments}
  ScalarParameter() {defaultValue=1, parameterName=add2mult}
  StaticSwitchParameter(True=Multiply(A=Multiply(A=@n37, B=@n15) {constB=0}) {constB=10}, False=Constant()) {parameterName=Distortion}
  n53 = Multiply(A=ScalarParameter() {parameterName=Speed}, B=OneMinus(None=Time()))
  n54 = ComponentMask(None=Multiply(A=TextureCoordinate() {coordinateIndex=1, uTiling=3}, B=ScalarParameter() {parameterName=Tile})) {r=true}
  n55 = Add(A=@n53, B=@n54)
  n56 = Multiply(A=@n53)
  n57 = Add(A=Add(A=Cosine(None=@n55), B=Cosine(None=@n55) {period=2}), B=Cosine(None=Add(A=@n54, B=@n56)) {period=0.75}) {desc=cosin}
  n58 = Add(A=@n54, B=@n53)
  n59 = Sine(None=@n58)
  n60 = Sine(None=@n58) {period=2}
  n61 = AppendVector(A=@n57, B=Add(A=Add(A=@n59, B=@n60), B=Sine(None=@n58) {period=0.75}) {desc=sin})
  n62 = Multiply(A=Multiply(A=AppendVector(A=@n61, B=@n57), B=ScalarParameter() {parameterName=Amount}), B=ScalarParameter() {parameterName=Intensity})
  StaticSwitchParameter(True=Multiply(A=@n39, B=Transform(None=@n62)) {constA=10}, False=Constant()) {parameterName=WPO}
 ]
}
